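# Animation Blueprint for the F_MED_Aztec body cosmetic mesh, which follows its attached parent mesh and adjusts two spine bones while targeting.
# Copy Pose From Mesh caches the parent's pose and curves.
# A branch applies additive yaw rotations to spine_04_a and spine_05_a, then layers them onto the cached pose with a per-bone blend.
# bIsTargeting selects between this adjusted pose and the plain copy; the event graph has no logic.
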AnimBlueprint /BRCosmetics/Characters/Player/Female/Medium/Bodies/F_MED_Aztec/Meshes/F_MED_Aztec_Body_AnimBP
target_skeleton=&/F_MED_Aztec_Body_Skeleton
graphs [
 AnimGraph {nodes [
   n1 = "Transform (Modify) Bone - Bone: spine_04_a"!(ComponentPose=Local To Component(LocalPose=Use cached pose 'CPFM Cache'()), Alpha=1, Rotation="0,-60,0") -> (Pose) {rotationMode=BMM_Additive, alphaBoolBlend {blendOption=Linear}, boneToModify {boneName=spine_04_a}, rotation {yaw=-60}}
   n2 = "Transform (Modify) Bone - Bone: spine_05_a"!(ComponentPose=@n1, Alpha=1, Rotation="0,-60,0") -> (Pose) {rotationMode=BMM_Additive, alphaBoolBlend {blendOption=Linear}, boneToModify {boneName=spine_05_a}, rotation {yaw=-60}}
   n3 = Component To Local(ComponentPose=@n2)
   n4 = Use cached pose 'CPFM Cache'()
   Output Pose(Result=Blend Poses by bool(
     BlendPose_0=Layered blend per bone(BasePose=@n4, BlendPoses_0=@n3) {layerSetup [{branchFilters [{blendDepth=1, boneName=spine_04_a}, {blendDepth=1, boneName=spine_05_a}]}]}
     BlendPose_1=Use cached pose 'CPFM Cache'()
     bActiveValue=Get bIsTargeting()
    ) {blendType=Linear}) {layerGroup=None}
   Save cached pose 'CPFM Cache'(Pose=Copy Pose From Mesh() {bCopyCurves=true, bUseAttachedParent=true}) {cacheName=CPFM Cache}
  ]}
 EventGraph {nodes [Event BlueprintUpdateAnimation()]}
]
variables []
