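# Physics asset for the F_MED_Aztec female player body cosmetic, covering the upper body and dynamic hair bones.
# Capsules on spine_05, clavicles, neck and head form the body shapes, joined by angularly free constraints.
# Box bodies form four hair chains from the head: three-bone left and right chains with tight twist limits, and two-bone side chains with free rotation.
# The left side chain's second bone is named dyn_hair_02_side, without the side letter.
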PhysicsAsset /BRCosmetics/Characters/Player/Female/Medium/Bodies/F_MED_Aztec/Meshes/F_MED_Aztec_Physics
bodies {
 clavicle_l [{shapeType=Capsule, length=16.25, radius=5.13, center=<8.2064 0.0005 0.9412>, rotation=<rot 79.9932 179.9999 179.9999>}]
 clavicle_r [{shapeType=Capsule, length=16.25, radius=5.13, center=<-8.7215 -0.615 -1.6969>, rotation=<rot 74.9872 180 5>}]
 dyn_hair_01_l [{shapeType=Box, extentX=10, extentY=2, extentZ=10, center=<5 0 0>}]
 dyn_hair_01_l_side [{shapeType=Box, extentX=10, extentY=2, extentZ=10, center=<5 0 0>}]
 dyn_hair_01_r [{shapeType=Box, extentX=10, extentY=2, extentZ=10, center=<5 0 0>}]
 dyn_hair_01_r_side [{shapeType=Box, extentX=10, extentY=2, extentZ=10, center=<5 0 0>}]
 dyn_hair_02_l [{shapeType=Box, extentX=10, extentY=2, extentZ=10, center=<5 0 0>}]
 dyn_hair_02_r [{shapeType=Box, extentX=10, extentY=2, extentZ=10, center=<5 0 0>}]
 dyn_hair_02_r_side [{shapeType=Box, extentX=4, extentY=2, extentZ=10, center=<3.032 -0.0001 -0.6349>, rotation=<rot -14.9999 0 0>}]
 dyn_hair_02_side [{shapeType=Box, extentX=4, extentY=2, extentZ=10, center=<3.0138 -0.0001 0.7791>, rotation=<rot 14.9999 0 0>}]
 dyn_hair_03_l [{shapeType=Box, extentX=10, extentY=2, extentZ=10, center=<5 0 0>}]
 dyn_hair_03_r [{shapeType=Box, extentX=10, extentY=2, extentZ=10, center=<5 0 0>}]
 head [{shapeType=Capsule, length=3.75, radius=7.875, center=<1.0589 1.6885 0>, rotation=<rot 90 0 0>}]
 neck_01 [{shapeType=Capsule, length=5.25, radius=2.75, center=<1.3587 0.0717 0>}]
 neck_02 [{shapeType=Capsule, length=0, radius=4.4279, center=<3.169 -0.4539 0>, rotation=<rot 90 0 0>}]
 spine_05 [{shapeType=Capsule, length=3.8108, radius=10}]
}
constraints [
 {
  bone1Name=head
  bone2Name=neck_02
  swing1LimitDegrees=45
  swing1Motion=Free
  swing2LimitDegrees=45
  swing2Motion=Free
  twistLimitDegrees=45
  twistMotion=Free
 }
 {
  bone1Name=neck_02
  bone2Name=neck_01
  swing1LimitDegrees=45
  swing1Motion=Free
  swing2LimitDegrees=45
  swing2Motion=Free
  twistLimitDegrees=45
  twistMotion=Free
 }
 {
  bone1Name=neck_01
  bone2Name=spine_05
  swing1LimitDegrees=45
  swing1Motion=Free
  swing2LimitDegrees=45
  swing2Motion=Free
  twistLimitDegrees=45
  twistMotion=Free
 }
 {
  bone1Name=clavicle_r
  bone2Name=spine_05
  swing1LimitDegrees=45
  swing1Motion=Free
  swing2LimitDegrees=45
  swing2Motion=Free
  twistLimitDegrees=45
  twistMotion=Free
 }
 {
  bone1Name=clavicle_l
  bone2Name=spine_05
  swing1LimitDegrees=45
  swing1Motion=Free
  swing2LimitDegrees=45
  swing2Motion=Free
  twistLimitDegrees=45
  twistMotion=Free
 }
 {bone1Name=dyn_hair_01_l, bone2Name=head, swing1LimitDegrees=45, swing2LimitDegrees=20, twistLimitDegrees=5}
 {bone1Name=dyn_hair_02_l, bone2Name=dyn_hair_01_l, swing1LimitDegrees=45, swing2LimitDegrees=20, twistLimitDegrees=5}
 {bone1Name=dyn_hair_03_l, bone2Name=dyn_hair_02_l, swing1LimitDegrees=45, swing2LimitDegrees=20, twistLimitDegrees=5}
 {bone1Name=dyn_hair_01_r, bone2Name=head, swing1LimitDegrees=45, swing2LimitDegrees=20, twistLimitDegrees=5}
 {bone1Name=dyn_hair_02_r, bone2Name=dyn_hair_01_r, swing1LimitDegrees=45, swing2LimitDegrees=20, twistLimitDegrees=5}
 {bone1Name=dyn_hair_03_r, bone2Name=dyn_hair_02_r, swing1LimitDegrees=45, swing2LimitDegrees=20, twistLimitDegrees=5}
 {
  bone1Name=dyn_hair_01_r_side
  bone2Name=head
  swing1LimitDegrees=45
  swing1Motion=Free
  swing2LimitDegrees=45
  swing2Motion=Free
  twistLimitDegrees=45
  twistMotion=Free
 }
 {
  bone1Name=dyn_hair_02_r_side
  bone2Name=dyn_hair_01_r_side
  swing1LimitDegrees=45
  swing1Motion=Free
  swing2LimitDegrees=45
  swing2Motion=Free
  twistLimitDegrees=45
  twistMotion=Free
 }
 {
  bone1Name=dyn_hair_01_l_side
  bone2Name=head
  swing1LimitDegrees=45
  swing1Motion=Free
  swing2LimitDegrees=45
  swing2Motion=Free
  twistLimitDegrees=45
  twistMotion=Free
 }
 {
  bone1Name=dyn_hair_02_side
  bone2Name=dyn_hair_01_l_side
  swing1LimitDegrees=45
  swing1Motion=Free
  swing2LimitDegrees=45
  swing2Motion=Free
  twistLimitDegrees=45
  twistMotion=Free
 }
]
