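# Post-process style hair AnimBP for the F_MED_Aztec face accessory, copying the parent mesh's pose and adding secondary motion.
# bRigidBodyEnabled selects RigidBody simulation with skydive-driven gravity and force, or spring-based AnimDynamics chains.
# Look At, constraint and bone-offset correctives then shape the hair, partly weighted by head rotation while targeting.
# The EventGraph computes those weights and forces, and resets dynamics after skydiving, parachute and crouch transitions.
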AnimBlueprint /BRCosmetics/Characters/Player/Female/Medium/Bodies/F_MED_Aztec/Meshes/F_MED_Aztec_Face_Acc_AnimBP
target_skeleton=&/F_MED_Aztec_Face_Acc_Skeleton
graphs [
 AnimGraph {nodes [
   n1 = Look At - Bone: dyn_hair_r_01(Alpha=1) {
    alphaBoolBlend {blendOption=Linear}
    alphaScaleBiasClamp {bClampResult=true, bInterpResult=true, clampMax=0.85, interpSpeedDecreasing=5, interpSpeedIncreasing=7}
    boneToModify {boneName=dyn_hair_r_01}
    lookAtLocation=<-6.5721 -1.8957 -7.0157>
    lookAtTarget {boneReference {boneName=clavicle_r}}
    lookAt_Axis {axis=<1 0>}
   }
   n2 = Look At - Bone: dyn_hair_bk_r_01(ComponentPose=@n1, Alpha=1) {
    alphaBoolBlend {blendOption=Linear}
    alphaScaleBiasClamp {bClampResult=true, bInterpResult=true, clampMax=0.85, interpSpeedDecreasing=5, interpSpeedIncreasing=7}
    boneToModify {boneName=dyn_hair_bk_r_01}
    lookAtLocation=<-6.6228 5.0192 -8.0944>
    lookAtTarget {boneReference {boneName=clavicle_r}}
    lookAt_Axis {axis=<1 0>}
   }
   n3 = Look At - Bone: dyn_hair_bk_01(ComponentPose=@n2, Alpha=1) {
    alphaBoolBlend {blendOption=Linear}
    alphaScaleBiasClamp {bClampResult=true, bInterpResult=true, clampMax=0.85, interpSpeedDecreasing=5, interpSpeedIncreasing=7}
    boneToModify {boneName=dyn_hair_bk_01}
    lookAtLocation=<-19.509 -4.1054 0.8885>
    lookAtTarget {boneReference {boneName=neck_01}}
    lookAt_Axis {axis=<1 0>}
   }
   n4 = Look At - Bone: dyn_hair_l_01(ComponentPose=@n3, Alpha=1) {
    alphaBoolBlend {blendOption=Linear}
    alphaScaleBiasClamp {bClampResult=true, bInterpResult=true, clampMax=0.85, interpSpeedDecreasing=5, interpSpeedIncreasing=7}
    boneToModify {boneName=dyn_hair_l_01}
    lookAtLocation=<7.5163 -1.1637 1.5828>
    lookAtTarget {boneReference {boneName=clavicle_l}}
    lookAt_Axis {axis=<1 0>}
   }
   Look At - Bone: dyn_hair_bk_l_01(ComponentPose=@n4, Alpha=1) {
    alphaBoolBlend {blendOption=Linear}
    alphaScaleBiasClamp {bClampResult=true, bInterpResult=true, clampMax=0.85, interpSpeedDecreasing=5, interpSpeedIncreasing=7}
    boneToModify {boneName=dyn_hair_bk_l_01}
    lookAtLocation=<4.049 -11.2733 -5.879>
    lookAtTarget {boneReference {boneName=clavicle_l}}
    lookAt_Axis {axis=<1 0>}
   }
   RigidBody() {
    bOverrideWorldGravity=true
    overridePhysicsAsset=&/F_MED_Aztec_Physics
    simulationSpace=BaseBoneSpace
    alphaBoolBlend {blendOption=Linear}
    baseBoneRef {boneName=spine_05}
    componentAppliedLinearAccClamp=<1225 1225 1225>
    componentLinearAccScale=<0.45 0.25 0.25>
    componentLinearVelScale=<0.45 0.25 0.25>
    overrideWorldGravity {z=-980}
   }
   n5 = Look At - Bone: dyn_hair_r_01(
    ComponentPose=Local To Component(LocalPose=Copy Pose From Mesh() {bCopyCurves=true, bUseAttachedParent=true})
    Alpha=Get zeroAnimDyn() -> (zeroAnimDyn:double)
   ) {
    alphaBoolBlend {blendOption=Linear}
    alphaScaleBiasClamp {bInterpResult=true, interpSpeedDecreasing=5, interpSpeedIncreasing=7}
    boneToModify {boneName=dyn_hair_r_01}
    lookAtLocation=<-25.1702 -5.8442 -7.3824>
    lookAtTarget {boneReference {boneName=clavicle_r}}
    lookAt_Axis {axis=<1 0>}
   }
   n6 = Look At - Bone: dyn_hair_bk_r_01(ComponentPose=@n5, Alpha=1) {
    alphaBoolBlend {blendOption=Linear}
    alphaScaleBiasClamp {bInterpResult=true, interpSpeedDecreasing=5, interpSpeedIncreasing=7}
    boneToModify {boneName=dyn_hair_bk_r_01}
    lookAtLocation=<-11.6381 6.5228 -8.3745>
    lookAtTarget {boneReference {boneName=clavicle_r}}
    lookAt_Axis {axis=<1 0>}
   }
   n7 = Look At - Bone: dyn_hair_bk_01(ComponentPose=@n6, Alpha=Get halfAnimDyn() -> (halfAnimDyn:double)) {
    alphaBoolBlend {blendOption=Linear}
    alphaScaleBiasClamp {bInterpResult=true, interpSpeedDecreasing=5, interpSpeedIncreasing=7}
    boneToModify {boneName=dyn_hair_bk_01}
    lookAtLocation=<-10.2838 -24.9354 -9.4362>
    lookAtTarget {boneReference {boneName=neck_01}}
    lookAt_Axis {axis=<1 0>}
   }
   n8 = Look At - Bone: dyn_hair_l_01(ComponentPose=@n7, Alpha=Get zeroAnimDyn()) {
    alphaBoolBlend {blendOption=Linear}
    alphaScaleBiasClamp {bInterpResult=true, interpSpeedDecreasing=5, interpSpeedIncreasing=7}
    boneToModify {boneName=dyn_hair_l_01}
    lookAtLocation=<29.1404 -12.4927 1.6218>
    lookAtTarget {boneReference {boneName=clavicle_l}}
    lookAt_Axis {axis=<1 0>}
   }
   n9 = Look At - Bone: dyn_hair_bk_l_01(ComponentPose=@n8, Alpha=Get halfAnimDyn()) {
    alphaBoolBlend {blendOption=Linear}
    alphaScaleBiasClamp {bInterpResult=true, interpSpeedDecreasing=5, interpSpeedIncreasing=7}
    boneToModify {boneName=dyn_hair_bk_l_01}
    lookAtLocation=<-1.6217 -19.4831 7.7851>
    lookAtTarget {boneReference {boneName=clavicle_l}}
    lookAt_Axis {axis=<1 0>}
   }
   n10 = Anim Dynamics - Bone: dyn_hair_l_01(ComponentPose=@n9, Alpha=1, ExternalForce="0,0,0") {
    angularDampingOverride=0.9999
    angularSpringConstant=500
    bAngularSpring=true
    bGravityOverrideInSimSpace=true
    bOverrideAngularDamping=true
    simulationSpace=BoneRelative
    alphaBoolBlend {blendOption=Linear}
    boundBone {boneName=dyn_hair_l_01}
    componentAppliedLinearAccClamp=<3500 3500 3500>
    componentLinearAccScale=<1.2 1.2 1.2>
    componentLinearVelScale=<1.2 1.2 1.2>
    physicsBodyDefinitions [{
      sphereCollisionRadius=0
      boundBone {boneName=dyn_hair_l_01}
      boxExtents=<15 10 10>
      constraintSetup {angularLimitsMax=<1 5 20>, angularLimitsMin=<-1 -45 -10>, angularTarget=<0 0 0>}
      localJointOffset=<5 0 0>
     }]
    relativeSpaceBone {boneName=root}
   }
   n11 = Anim Dynamics - Bone: dyn_hair_l_02(ComponentPose=@n10, Alpha=1, ExternalForce="0,0,0") {
    angularDampingOverride=0.9999
    angularSpringConstant=500
    bAngularSpring=true
    bGravityOverrideInSimSpace=true
    bOverrideAngularDamping=true
    simulationSpace=BoneRelative
    alphaBoolBlend {blendOption=Linear}
    boundBone {boneName=dyn_hair_l_02}
    componentAppliedLinearAccClamp=<3500 3500 3500>
    componentLinearAccScale=<1.2 1.2 1.2>
    componentLinearVelScale=<1.2 1.2 1.2>
    physicsBodyDefinitions [{
      sphereCollisionRadius=0
      boundBone {boneName=dyn_hair_l_02}
      boxExtents=<15 10 10>
      constraintSetup {angularLimitsMax=<1 5 20>, angularLimitsMin=<-1 -45 -10>, angularTarget=<0 0 0>}
      localJointOffset=<5 0 0>
     }]
    relativeSpaceBone {boneName=root}
   }
   n12 = Anim Dynamics - Bone: dyn_hair_l_03(ComponentPose=@n11, Alpha=1, ExternalForce="0,0,0") {
    angularDampingOverride=0.9999
    angularSpringConstant=500
    bAngularSpring=true
    bGravityOverrideInSimSpace=true
    bOverrideAngularDamping=true
    simulationSpace=BoneRelative
    alphaBoolBlend {blendOption=Linear}
    boundBone {boneName=dyn_hair_l_03}
    componentAppliedLinearAccClamp=<3500 3500 3500>
    componentLinearAccScale=<1.2 1.2 1.2>
    componentLinearVelScale=<1.2 1.2 1.2>
    physicsBodyDefinitions [{
      sphereCollisionRadius=0
      boundBone {boneName=dyn_hair_l_03}
      boxExtents=<15 10 10>
      constraintSetup {angularLimitsMax=<1 5 20>, angularLimitsMin=<-1 -45 -10>, angularTarget=<0 0 0>}
      localJointOffset=<5 0 0>
     }]
    relativeSpaceBone {boneName=root}
   }
   n13 = Anim Dynamics - Bone: dyn_hair_bk_l_01(ComponentPose=@n12, Alpha=1, ExternalForce="0,0,0") {
    angularDampingOverride=0.9999
    angularSpringConstant=500
    bAngularSpring=true
    bGravityOverrideInSimSpace=true
    bOverrideAngularDamping=true
    simulationSpace=BoneRelative
    alphaBoolBlend {blendOption=Linear}
    boundBone {boneName=dyn_hair_bk_l_01}
    componentAppliedLinearAccClamp=<3500 3500 3500>
    componentLinearAccScale=<1.2 1.2 1.2>
    componentLinearVelScale=<1.2 1.2 1.2>
    physicsBodyDefinitions [{
      sphereCollisionRadius=0
      boundBone {boneName=dyn_hair_bk_l_01}
      boxExtents=<15 10 10>
      constraintSetup {angularLimitsMax=<1 5 15>, angularLimitsMin=<-1 -45 -15>, angularTarget=<0 0 0>}
      localJointOffset=<5 0 0>
     }]
    relativeSpaceBone {boneName=root}
   }
   n14 = Anim Dynamics - Bone: dyn_hair_bk_l_02(ComponentPose=@n13, Alpha=1, ExternalForce="0,0,0") {
    angularDampingOverride=0.9999
    angularSpringConstant=500
    bAngularSpring=true
    bGravityOverrideInSimSpace=true
    bOverrideAngularDamping=true
    simulationSpace=BoneRelative
    alphaBoolBlend {blendOption=Linear}
    boundBone {boneName=dyn_hair_bk_l_02}
    componentAppliedLinearAccClamp=<3500 3500 3500>
    componentLinearAccScale=<1.2 1.2 1.2>
    componentLinearVelScale=<1.2 1.2 1.2>
    physicsBodyDefinitions [{
      sphereCollisionRadius=0
      boundBone {boneName=dyn_hair_bk_l_02}
      boxExtents=<15 10 10>
      constraintSetup {angularLimitsMax=<1 5 15>, angularLimitsMin=<-1 -45 -15>, angularTarget=<0 0 0>}
      localJointOffset=<5 0 0>
     }]
    relativeSpaceBone {boneName=root}
   }
   n15 = Anim Dynamics - Bone: dyn_hair_bk_l_03(ComponentPose=@n14, Alpha=1, ExternalForce="0,0,0") {
    angularDampingOverride=0.9999
    angularSpringConstant=500
    bAngularSpring=true
    bGravityOverrideInSimSpace=true
    bOverrideAngularDamping=true
    simulationSpace=BoneRelative
    alphaBoolBlend {blendOption=Linear}
    boundBone {boneName=dyn_hair_bk_l_03}
    componentAppliedLinearAccClamp=<3500 3500 3500>
    componentLinearAccScale=<1.2 1.2 1.2>
    componentLinearVelScale=<1.2 1.2 1.2>
    physicsBodyDefinitions [{
      sphereCollisionRadius=0
      boundBone {boneName=dyn_hair_bk_l_03}
      boxExtents=<15 10 10>
      constraintSetup {angularLimitsMax=<1 0 10>, angularLimitsMin=<-1 -35 -10>, angularTarget=<0 0 0>}
      localJointOffset=<5 0 0>
     }]
    relativeSpaceBone {boneName=root}
   }
   n16 = Anim Dynamics - Bone: dyn_hair_bk_01(ComponentPose=@n15, Alpha=1, ExternalForce="0,0,0") {
    angularDampingOverride=0.9999
    angularSpringConstant=100
    bAngularSpring=true
    bGravityOverrideInSimSpace=true
    bOverrideAngularDamping=true
    simulationSpace=BoneRelative
    alphaBoolBlend {blendOption=Linear}
    boundBone {boneName=dyn_hair_bk_01}
    componentAppliedLinearAccClamp=<3500 3500 3500>
    componentLinearAccScale=<1.2 1.2 1.2>
    componentLinearVelScale=<1.2 1.2 1.2>
    physicsBodyDefinitions [{
      sphereCollisionRadius=0
      boundBone {boneName=dyn_hair_bk_01}
      boxExtents=<20 10 10>
      constraintSetup {angularLimitsMax=<1 10 15>, angularLimitsMin=<-1 -30 -15>, angularTarget=<0 0 0>}
      localJointOffset=<2 0 0>
     }]
    relativeSpaceBone {boneName=root}
   }
   n17 = Anim Dynamics - Bone: dyn_hair_bk_02(ComponentPose=@n16, Alpha=1, ExternalForce="0,0,0") {
    angularDampingOverride=0.9999
    angularSpringConstant=100
    bAngularSpring=true
    bGravityOverrideInSimSpace=true
    bOverrideAngularDamping=true
    simulationSpace=BoneRelative
    alphaBoolBlend {blendOption=Linear}
    boundBone {boneName=dyn_hair_bk_02}
    componentAppliedLinearAccClamp=<3500 3500 3500>
    componentLinearAccScale=<1.2 1.2 1.2>
    componentLinearVelScale=<1.2 1.2 1.2>
    physicsBodyDefinitions [{
      sphereCollisionRadius=0
      boundBone {boneName=dyn_hair_bk_02}
      boxExtents=<20 10 10>
      constraintSetup {angularLimitsMax=<1 10 15>, angularLimitsMin=<-1 -30 -15>, angularTarget=<0 0 0>}
      localJointOffset=<2 0 0>
     }]
    relativeSpaceBone {boneName=root}
   }
   n18 = Anim Dynamics - Bone: dyn_hair_bk_03(ComponentPose=@n17, Alpha=1, ExternalForce="0,0,0") {
    angularDampingOverride=0.9999
    angularSpringConstant=100
    bAngularSpring=true
    bGravityOverrideInSimSpace=true
    bOverrideAngularDamping=true
    simulationSpace=BoneRelative
    alphaBoolBlend {blendOption=Linear}
    boundBone {boneName=dyn_hair_bk_03}
    componentAppliedLinearAccClamp=<3500 3500 3500>
    componentLinearAccScale=<1.2 1.2 1.2>
    componentLinearVelScale=<1.2 1.2 1.2>
    physicsBodyDefinitions [{
      sphereCollisionRadius=0
      boundBone {boneName=dyn_hair_bk_03}
      boxExtents=<20 10 10>
      constraintSetup {angularLimitsMax=<1 5 15>, angularLimitsMin=<-1 -30 -15>, angularTarget=<0 0 0>}
      localJointOffset=<2 0 0>
     }]
    relativeSpaceBone {boneName=root}
   }
   n19 = Anim Dynamics - Bone: dyn_hair_r_01(ComponentPose=@n18, Alpha=1, ExternalForce="0,0,0") {
    angularDampingOverride=0.9999
    angularSpringConstant=100
    bAngularSpring=true
    bGravityOverrideInSimSpace=true
    bOverrideAngularDamping=true
    simulationSpace=BoneRelative
    alphaBoolBlend {blendOption=Linear}
    boundBone {boneName=dyn_hair_r_01}
    componentAppliedLinearAccClamp=<3500 3500 3500>
    componentLinearAccScale=<1.2 1.2 1.2>
    componentLinearVelScale=<1.2 1.2 1.2>
    physicsBodyDefinitions [{
      sphereCollisionRadius=0
      boundBone {boneName=dyn_hair_r_01}
      boxExtents=<20 10 10>
      constraintSetup {angularLimitsMax=<1 5 15>, angularLimitsMin=<-1 -30 -15>, angularTarget=<0 0 0>}
      localJointOffset=<2 0 0>
     }]
    relativeSpaceBone {boneName=root}
   }
   n20 = Anim Dynamics - Bone: dyn_hair_r_02(ComponentPose=@n19, Alpha=1, ExternalForce="0,0,0") {
    angularDampingOverride=0.9999
    angularSpringConstant=100
    bAngularSpring=true
    bGravityOverrideInSimSpace=true
    bOverrideAngularDamping=true
    simulationSpace=BoneRelative
    alphaBoolBlend {blendOption=Linear}
    boundBone {boneName=dyn_hair_r_02}
    componentAppliedLinearAccClamp=<3500 3500 3500>
    componentLinearAccScale=<1.2 1.2 1.2>
    componentLinearVelScale=<1.2 1.2 1.2>
    physicsBodyDefinitions [{
      sphereCollisionRadius=0
      boundBone {boneName=dyn_hair_r_02}
      boxExtents=<20 10 10>
      constraintSetup {angularLimitsMax=<1 5 15>, angularLimitsMin=<-1 -30 -15>, angularTarget=<0 0 0>}
      localJointOffset=<2 0 0>
     }]
    relativeSpaceBone {boneName=root}
   }
   n21 = Anim Dynamics - Bone: dyn_hair_r_03(ComponentPose=@n20, Alpha=1, ExternalForce="0,0,0") {
    angularDampingOverride=0.9999
    angularSpringConstant=100
    bAngularSpring=true
    bGravityOverrideInSimSpace=true
    bOverrideAngularDamping=true
    simulationSpace=BoneRelative
    alphaBoolBlend {blendOption=Linear}
    boundBone {boneName=dyn_hair_r_03}
    componentAppliedLinearAccClamp=<3500 3500 3500>
    componentLinearAccScale=<1.2 1.2 1.2>
    componentLinearVelScale=<1.2 1.2 1.2>
    physicsBodyDefinitions [{
      sphereCollisionRadius=0
      boundBone {boneName=dyn_hair_r_03}
      boxExtents=<20 10 10>
      constraintSetup {angularLimitsMax=<1 5 15>, angularLimitsMin=<-1 -30 -15>, angularTarget=<0 0 0>}
      localJointOffset=<2 0 0>
     }]
    relativeSpaceBone {boneName=root}
   }
   n22 = Anim Dynamics - Bone: dyn_hair_bk_r_01(ComponentPose=@n21, Alpha=1, ExternalForce="0,0,0") {
    angularDampingOverride=0.9999
    angularSpringConstant=100
    bAngularSpring=true
    bGravityOverrideInSimSpace=true
    bOverrideAngularDamping=true
    simulationSpace=BoneRelative
    alphaBoolBlend {blendOption=Linear}
    boundBone {boneName=dyn_hair_bk_r_01}
    componentAppliedLinearAccClamp=<3500 3500 3500>
    componentLinearAccScale=<1.2 1.2 1.2>
    componentLinearVelScale=<1.2 1.2 1.2>
    physicsBodyDefinitions [{
      sphereCollisionRadius=0
      boundBone {boneName=dyn_hair_bk_r_01}
      boxExtents=<20 10 10>
      constraintSetup {angularLimitsMax=<1 5 15>, angularLimitsMin=<-1 -30 -15>, angularTarget=<0 0 0>}
      localJointOffset=<2 0 0>
     }]
    relativeSpaceBone {boneName=root}
   }
   n23 = Anim Dynamics - Bone: dyn_hair_bk_r_02(ComponentPose=@n22, Alpha=1, ExternalForce="0,0,0") {
    angularDampingOverride=0.9999
    angularSpringConstant=100
    bAngularSpring=true
    bGravityOverrideInSimSpace=true
    bOverrideAngularDamping=true
    simulationSpace=BoneRelative
    alphaBoolBlend {blendOption=Linear}
    boundBone {boneName=dyn_hair_bk_r_02}
    componentAppliedLinearAccClamp=<3500 3500 3500>
    componentLinearAccScale=<1.2 1.2 1.2>
    componentLinearVelScale=<1.2 1.2 1.2>
    physicsBodyDefinitions [{
      sphereCollisionRadius=0
      boundBone {boneName=dyn_hair_bk_r_02}
      boxExtents=<20 10 10>
      constraintSetup {angularLimitsMax=<1 5 15>, angularLimitsMin=<-1 -30 -15>, angularTarget=<0 0 0>}
      localJointOffset=<2 0 0>
     }]
    relativeSpaceBone {boneName=root}
   }
   n24 = Anim Dynamics - Bone: dyn_hair_bk_r_03(ComponentPose=@n23, Alpha=1, ExternalForce="0,0,0") {
    angularDampingOverride=0.9999
    angularSpringConstant=100
    bAngularSpring=true
    bGravityOverrideInSimSpace=true
    bOverrideAngularDamping=true
    simulationSpace=BoneRelative
    alphaBoolBlend {blendOption=Linear}
    boundBone {boneName=dyn_hair_bk_r_03}
    componentAppliedLinearAccClamp=<3500 3500 3500>
    componentLinearAccScale=<1.2 1.2 1.2>
    componentLinearVelScale=<1.2 1.2 1.2>
    physicsBodyDefinitions [{
      sphereCollisionRadius=0
      boundBone {boneName=dyn_hair_bk_r_03}
      boxExtents=<20 10 10>
      constraintSetup {angularLimitsMax=<1 5 15>, angularLimitsMin=<-1 -30 -15>, angularTarget=<0 0 0>}
      localJointOffset=<2 0 0>
     }]
    relativeSpaceBone {boneName=root}
   }
   n25 = Component To Local(ComponentPose=@n24)
   n26 = RigidBody!(
    ComponentPose=Local To Component(LocalPose=Copy Pose From Mesh() {bCopyCurves=true, bUseAttachedParent=true})
    Alpha
    OverrideWorldGravity:Vector=SelectVector(A="0, 0,-100", B="0, 0,-980", bPickA=Get bIsSkydiveFloating())
    ExternalForce=SelectVector(A=Get skydiving vector force() -> (skydiving vector force:Vector), B="0, 0, 0", bPickA=Get bIsSkyDiving())
   ) -> (Pose) {
    bClampLinearTranslationLimitToRefPose=true
    bOverrideWorldGravity=true
    overridePhysicsAsset=&/F_MED_Aztec_FaceAcc_Physics
    simulationSpace=BaseBoneSpace
    alphaBoolBlend {blendOption=Linear}
    baseBoneRef {boneName=head}
    componentLinearAccScale=<0.2 0.5 0.8>
    componentLinearVelScale=<0.2 0.5 0.8>
    overrideWorldGravity {z=-980}
   }
   n27 = Anim Dynamics - Bone: dyn_bang(ComponentPose=@n26, Alpha=1, ExternalForce="0,0,0") {
    angularDampingOverride=0.97
    bGravityOverrideInSimSpace=true
    bOverrideAngularDamping=true
    bOverrideLinearDamping=true
    lODThreshold=1
    linearDampingOverride=0.97
    numSolverIterationsPostUpdate=4
    numSolverIterationsPreUpdate=16
    alphaBoolBlend {blendOption=Linear}
    boundBone {boneName=dyn_bang}
    componentLinearAccScale=<1 1 1>
    componentLinearVelScale=<1 1 1>
    physicsBodyDefinitions [{
      sphereCollisionRadius=0
      boundBone {boneName=dyn_bang}
      boxExtents=<5 5 5>
      constraintSetup {angularTargetAxis=AxisZ, twistAxis=AxisZ, angularLimitsMax=<0 5 0>, angularLimitsMin=<0 -45 0>, angularTarget=<0 0 0>}
      localJointOffset=<0 0 -0.1>
     }]
    relativeSpaceBone {boneName=head}
   }
   n28 = "Transform (Modify) Bone - Bone: dyn_hair_r_02"!(ComponentPose=@n27, Alpha=1, Translation="0,0,0", Rotation="-0.000,30.000,-0.000") -> (Pose) {
    rotationMode=BMM_Replace
    rotationSpace=BCS_BoneSpace
    scaleMode=BMM_Replace
    scaleSpace=BCS_BoneSpace
    translationMode=BMM_Replace
    translationSpace=BCS_BoneSpace
    alphaBoolBlend {blendOption=Linear}
    alphaScaleBiasClamp {bInterpResult=true, interpSpeedDecreasing=5, interpSpeedIncreasing=7}
    boneToModify {boneName=dyn_hair_r_02}
    rotation=<rot -0.0001 30 0>
   }
   n29 = Component To Local(ComponentPose=@n28)
   n30 = Get bRigidBodyEnabled()
   n31 = Look At - Bone: dyn_hair_r_01(
    ComponentPose=Local To Component(LocalPose=Blend Poses by bool(BlendPose_0=@n29, BlendPose_1=@n25, bActiveValue=@n30) {blendType=Linear})
    Alpha=1
   ) {
    alphaBoolBlend {blendOption=Linear}
    alphaScaleBiasClamp {bInterpResult=true, interpSpeedDecreasing=5, interpSpeedIncreasing=7}
    boneToModify {boneName=dyn_hair_r_01}
    lookAtLocation=<-3.352 -7.6915 11.2967>
    lookAtTarget {boneReference {boneName=neck_01}}
    lookAt_Axis {axis=<1 0>}
   }
   n32 = Look At - Bone: dyn_hair_bk_r_01(ComponentPose=@n31, Alpha=1) {
    alphaBoolBlend {blendOption=Linear}
    alphaScaleBiasClamp {bInterpResult=true, interpSpeedDecreasing=5, interpSpeedIncreasing=7}
    boneToModify {boneName=dyn_hair_bk_r_01}
    lookAtLocation=<0.0062 -8.7196 11.6223>
    lookAtTarget {boneReference {boneName=neck_01}}
    lookAt_Axis {axis=<1 0>}
   }
   n33 = Look At - Bone: dyn_hair_bk_01(ComponentPose=@n32, Alpha=1) {
    alphaBoolBlend {blendOption=Linear}
    alphaScaleBiasClamp {bInterpResult=true, interpSpeedDecreasing=5, interpSpeedIncreasing=7}
    boneToModify {boneName=dyn_hair_bk_01}
    lookAtLocation=<0.0024 -15.376 1.9583>
    lookAtTarget {boneReference {boneName=neck_01}}
    lookAt_Axis {axis=<1 0>}
   }
   n34 = Look At - Bone: dyn_hair_l_01(ComponentPose=@n33, Alpha=1) {
    alphaBoolBlend {blendOption=Linear}
    alphaScaleBiasClamp {bInterpResult=true, interpSpeedDecreasing=5, interpSpeedIncreasing=7}
    boneToModify {boneName=dyn_hair_l_01}
    lookAtLocation=<0.0038 -8.132 -9.6386>
    lookAtTarget {boneReference {boneName=neck_01}}
    lookAt_Axis {axis=<1 0>}
   }
   n35 = Look At - Bone: dyn_hair_bk_l_01(ComponentPose=@n34, Alpha=1) {
    alphaBoolBlend {blendOption=Linear}
    alphaScaleBiasClamp {bInterpResult=true, interpSpeedDecreasing=5, interpSpeedIncreasing=7}
    boneToModify {boneName=dyn_hair_bk_l_01}
    lookAtLocation=<0.0061 -15.8144 -4.7924>
    lookAtTarget {boneReference {boneName=neck_01}}
    lookAt_Axis {axis=<1 0>}
   }
   n36 = Get Head_up() -> (Head_up:double)
   n37 = "Constraint - dyn_hair_bk_r_02 (P)"(ComponentPose=@n35, Alpha=@n36, ConstraintWeights_0=1) {
    alphaBoolBlend {blendOption=Linear}
    alphaScaleBiasClamp {bInterpResult=true, interpSpeedDecreasing=5, interpSpeedIncreasing=7}
    boneToModify {boneName=dyn_hair_bk_r_02}
    constraintSetup [{targetBone {boneName=clavicle_r}}]
    constraintWeights [1]
   }
   n38 = "Transform (Modify) Bone - Bone: dyn_hair_r_01"!(ComponentPose=@n37, Alpha=@n36, Translation="-0.493,-0.563,-0.629", Rotation="42.362,-14.805,23.116") -> (Pose) {
    rotationMode=BMM_Replace
    rotationSpace=BCS_BoneSpace
    translationMode=BMM_Replace
    translationSpace=BCS_BoneSpace
    alphaBoolBlend {blendOption=Linear}
    alphaScaleBiasClamp {bInterpResult=true, interpSpeedDecreasing=5, interpSpeedIncreasing=7}
    boneToModify {boneName=dyn_hair_r_01}
    rotation=<rot 42.362 -14.8049 23.116>
    translation=<-0.493 -0.563 -0.629>
   }
   n39 = "Transform (Modify) Bone - Bone: dyn_hair_bk_r_01"!(ComponentPose=@n38, Alpha=@n36, Translation="-1.003,-2.943,-2.041", Rotation="40.146,62.410,63.827") -> (Pose) {
    rotationMode=BMM_Replace
    rotationSpace=BCS_BoneSpace
    translationMode=BMM_Replace
    translationSpace=BCS_BoneSpace
    alphaBoolBlend {blendOption=Linear}
    alphaScaleBiasClamp {bInterpResult=true, interpSpeedDecreasing=5, interpSpeedIncreasing=7}
    boneToModify {boneName=dyn_hair_bk_r_01}
    rotation=<rot 40.146 62.4101 63.827>
    translation=<-1.003 -2.943 -2.041>
   }
   n40 = "Transform (Modify) Bone - Bone: dyn_hair_bk_01"!(ComponentPose=@n39, Alpha=@n36, Scale="0.8,1.000,1.000") -> (Pose) {
    rotationSpace=BCS_BoneSpace
    scaleMode=BMM_Replace
    scaleSpace=BCS_BoneSpace
    translationSpace=BCS_BoneSpace
    alphaBoolBlend {blendOption=Linear}
    alphaScaleBiasClamp {bInterpResult=true, interpSpeedDecreasing=5, interpSpeedIncreasing=7}
    boneToModify {boneName=dyn_hair_bk_01}
    scale {x=0.8}
   }
   n41 = "Transform (Modify) Bone - Bone: dyn_hair_bk_r_01"(ComponentPose=@n40, Alpha=@n36, Translation="5,5,0", Rotation="0,0,0") {
    rotationMode=BMM_Replace
    rotationSpace=BCS_BoneSpace
    translationMode=BMM_Replace
    translationSpace=BCS_BoneSpace
    alphaBoolBlend {blendOption=Linear}
    alphaScaleBiasClamp {bInterpResult=true, interpSpeedDecreasing=5, interpSpeedIncreasing=7}
    boneToModify {boneName=dyn_hair_bk_r_01}
    translation=<5 5>
   }
   n42 = Get Head Down() -> (Head Down:double)
   n43 = "Constraint - dyn_hair_bk_r_02 (P)"(ComponentPose=@n41, Alpha=@n42, ConstraintWeights_0=1) {
    alphaBoolBlend {blendOption=Linear}
    alphaScaleBiasClamp {bInterpResult=true, interpSpeedDecreasing=5, interpSpeedIncreasing=7}
    boneToModify {boneName=dyn_hair_bk_r_02}
    constraintSetup [{targetBone {boneName=clavicle_r}}]
    constraintWeights [1]
   }
   n44 = "Constraint - dyn_hair_r_03 (P)"(ComponentPose=@n43, Alpha=@n42, ConstraintWeights_0=1) {
    alphaBoolBlend {blendOption=Linear}
    alphaScaleBiasClamp {bInterpResult=true, interpSpeedDecreasing=5, interpSpeedIncreasing=7}
    boneToModify {boneName=dyn_hair_r_03}
    constraintSetup [{targetBone {boneName=clavicle_r}}]
    constraintWeights [1]
   }
   n45 = "Transform (Modify) Bone - Bone: dyn_hair_r_01"(ComponentPose=@n44, Alpha=@n42, Translation="0,-1.3,1.5", Rotation="0,-10,0") {
    rotationMode=BMM_Replace
    rotationSpace=BCS_BoneSpace
    scaleMode=BMM_Replace
    scaleSpace=BCS_BoneSpace
    translationMode=BMM_Replace
    translationSpace=BCS_BoneSpace
    alphaBoolBlend {blendOption=Linear}
    alphaScaleBiasClamp {bInterpResult=true, interpSpeedDecreasing=5, interpSpeedIncreasing=7}
    boneToModify {boneName=dyn_hair_r_01}
    rotation {yaw=-10}
    translation {y=-1.3, z=1.5}
   }
   n46 = "Transform (Modify) Bone - Bone: dyn_hair_bk_r_01"!(ComponentPose=@n45, Alpha=@n42, Translation="5,0,4") -> (Pose) {
    rotationSpace=BCS_BoneSpace
    scaleSpace=BCS_BoneSpace
    translationMode=BMM_Replace
    translationSpace=BCS_BoneSpace
    alphaBoolBlend {blendOption=Linear}
    alphaScaleBiasClamp {bInterpResult=true, interpSpeedDecreasing=5, interpSpeedIncreasing=7}
    boneToModify {boneName=dyn_hair_bk_r_01}
    translation {x=5, z=4}
   }
   n47 = "Constraint - dyn_hair_bk_r_02 (P)"(ComponentPose=@n46, Alpha=1, ConstraintWeights_0=1) {
    alphaBoolBlend {blendOption=Linear}
    alphaScaleBiasClamp {bClampResult=true, bInterpResult=true, clampMax=0.65, interpSpeedDecreasing=5, interpSpeedIncreasing=7}
    boneToModify {boneName=dyn_hair_bk_r_02}
    constraintSetup [{targetBone {boneName=clavicle_r}}]
    constraintWeights [1]
   }
   n48 = "Constraint - dyn_hair_bk_02 (P)"(ComponentPose=@n47, Alpha=1, ConstraintWeights_0=1) {
    alphaBoolBlend {blendOption=Linear}
    alphaScaleBiasClamp {bClampResult=true, bInterpResult=true, clampMax=0.65, interpSpeedDecreasing=5, interpSpeedIncreasing=7}
    boneToModify {boneName=dyn_hair_bk_02}
    constraintSetup [{targetBone {boneName=spine_05}}]
    constraintWeights [1]
   }
   Output Pose(Result=Component To Local(ComponentPose=@n48)) {layerGroup=None}
  ]}
 EventGraph {nodes [
   Event BlueprintUpdateAnimation()
   > n1 = Sequence()
   > n2 = Branch(Condition=Get bIsSkydiveFloating())
   n3 = Get Direction Between Sockets!(
    Component:Skeletal Mesh Component Object Reference=GetOwnerSkeletalMeshComponent()
    SocketOrBoneNameFrom:Name=head
    SocketOrBoneNameTo:Name=pelvis
   ) -> (ReturnValue:Vector) {is=title K2Node_CallFunction}
   > Set skydiving vector force(skydiving vector force:Vector=vector + vector(
     A:Vector=vector * vector(A:Vector=@n3, B:Vector="75,75,75") -> (ReturnValue:Vector) {is=title K2Node_CommutativeAssociativeBinaryOperator}
     B:Vector="0, 0, 0"
    ) -> (ReturnValue:Vector) {is=title K2Node_CommutativeAssociativeBinaryOperator}) -> (Output_Get:Vector)
   n4 = Branch(execute=@n2.else, Condition=Get bIsSkydiveDiveMode())
   > Set skydiving vector force(skydiving vector force=vector + vector(A=vector * vector(A=@n3, B="180,180,200"), B="0, 0, 0"))
   Set skydiving vector force(execute=@n4.else, skydiving vector force=vector * vector(A=@n3, B="-15,-15,-15"))
   n5 = GetOwnerSkeletalMeshComponent()
   n6 = BreakRotator(InRot=BreakTransform(InTransform=GetDeltaTransformFromRefPose(self=@n5, BoneName=head, BaseName=spine_05)).Rotation)
   n7 = NOT Boolean(A=AND Boolean(A=Get bIsCrouching(), B=Get bIsMoving2D()))
   n8 = Get bIsTargeting()
   Set Head_up(
    execute=@n1.then_1
    Head_up:double=SelectFloat(A=MapRangeClamped(Value=@n6.Yaw, InRangeA=-3, InRangeB=-40), B=0, bPickA=AND Boolean(A=@n8, B=@n7))
   ) -> (Output_Get:double)
   > Set Head Down(Head Down:double=SelectFloat(A=MapRangeClamped(Value=@n6.Yaw, InRangeA=1, InRangeB=10), B=0, bPickA=Get bIsTargeting())) -> (Output_Get:double)
   > Set Head Left(Head Left:double=MapRangeClamped(Value=@n6.Pitch, InRangeA=1, InRangeB=-1)) -> (Output_Get:double)
   > Set Target_off(Target_off:double=SelectFloat(A=0, B=1, bPickA=Get bIsTargeting())) -> (Output_Get:double)
   n9 = Get bAnimDynamicsEnabled() -> (bAnimDynamicsEnabled:bool)
   > Set zeroAnimDyn(zeroAnimDyn:double=SelectFloat(A=1, B=0, bPickA=@n9)) -> (Output_Get:double)
   > Set halfAnimDyn(halfAnimDyn:double=SelectFloat(A=1, B=0.5, bPickA=@n9)) -> (Output_Get:double)
   Event OnParachuteOpened()
   > n10 = Delay(Duration=1.5)
   ResetDynamics(execute=[@n10.then, @n12.then, @n17.then], InTeleportType=TeleportPhysics)
   > ResetDynamics(InTeleportType=ResetPhysics)
   n11 = Event OnBeginSkydiving()
   n12 = Delay(execute=[@n11.then, @n13.then, @n14.then, @n15.then], Duration=0.125)
   n13 = Event OnEndSkydiving()
   n14 = Event OnBeginCrouching()
   n15 = Event OnEndCrouching()
   n16 = Get bIsCrouching()
   n17 = Branch(Condition=AND Boolean(A=Get bIsTargeting(), B=NOT Boolean(A=@n16)))
   Event OnBeginTargeting() {is=title K2Node_Event}
  ]}
]
variables [
 halfAnimDyn {type=number}
 Head Down {type=number}
 Head Left {type=number}
 Head_up {type=number}
 Skel_Mesh {type=SkeletalMeshComponent}
 skydiving vector force {type=Vector}
 Target_off {type=number}
 zeroAnimDyn {type=number}
]
